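# Sound cue for a blocked hammer defense in melee combat.
# A random node picks one of three equally weighted wave variations and feeds it through a modulator at the root.
# The cue plays in the melee weapon sound class with close-range stereo melee attenuation, a raised volume multiplier and a shared defensive melee concurrency limit.
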SoundCue /CRD_Melee/Audio/Sounds/Hammer/Melee_Hammer_Defense_Blocked_Cue
root=Modulator_0
nodes [
 {id=Modulator_0, children [Random_0]}
 {id=Random_0, children [WavePlayer_0,WavePlayer_1,WavePlayer_2], properties {weights [1,1,1]}}
 {id=WavePlayer_0, properties {soundWaveAssetPtr=&/Melee_Hammer_Defense_Blocked_01}}
 {id=WavePlayer_1, properties {soundWaveAssetPtr=&/Melee_Hammer_Defense_Blocked_02}}
 {id=WavePlayer_2, properties {soundWaveAssetPtr=&/Melee_Hammer_Defense_Blocked_03}}
]
properties {
 attenuationSettings=/Game/Sounds/Attenuation/Fort_Weapon_Melee_Attack_Close_Range_Stereo_Att
 maxDistance=10058.75
 soundClassObject=&Fort_WeaponMelee
 volumeMultiplier=1.3
 concurrencySet [/CRD_Melee/Audio/Settings/Melee_Defensive_Global_Conc]
}
