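# Sound cue for the hammer's defensive parry, routed through the melee weapon sound class with close-range stereo melee attenuation.
# A random node picks one of three equally weighted parry waves and feeds a modulator at the root.
# Overridden concurrency allows one instance per owner with a short retrigger window, and a global melee defensive concurrency set is also assigned.
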SoundCue /CRD_Melee/Audio/Sounds/Hammer/Melee_Hammer_Defense_Parry_Cue
root=Modulator_0
nodes [
 {id=Modulator_0, children [Random_1]}
 {id=Random_1, children [WavePlayer_3,WavePlayer_4,WavePlayer_5], properties {weights [1,1,1]}}
 {id=WavePlayer_3, properties {soundWaveAssetPtr=&/Melee_Hammer_Defense_Parry_01}}
 {id=WavePlayer_4, properties {soundWaveAssetPtr=&/Melee_Hammer_Defense_Parry_02}}
 {id=WavePlayer_5, properties {soundWaveAssetPtr=&/Melee_Hammer_Defense_Parry_03}}
]
properties {
 attenuationSettings=/Game/Sounds/Attenuation/Fort_Weapon_Melee_Attack_Close_Range_Stereo_Att
 bOverrideConcurrency=true
 maxDistance=10058.75
 soundClassObject=&Fort_WeaponMelee
 volumeMultiplier=1.2
 concurrencyOverrides {bLimitToOwner=true, maxCount=1, retriggerTime=0.2, voiceStealReleaseTime=0.1, volumeScaleAttackTime=0.1}
 concurrencySet [/CRD_Melee/Audio/Settings/Melee_Defensive_Global_Conc]
}
