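# Sound cue for the drop sound of the melee hammer weapon.
# A Random node picks one of three equally weighted drop wave variations and passes it through a Modulator at the root for variation.
# The cue uses the weapon melee sound class and a shared weapon drop attenuation, with a reduced volume multiplier.
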SoundCue /CRD_Melee/Audio/Sounds/Hammer/Melee_Hammer_Drop_Cue
root=Modulator_0
nodes [
 {id=Modulator_0, children [Random_0]}
 {id=Random_0, children [WavePlayer_0,WavePlayer_1,WavePlayer_2], properties {weights [1,1,1]}}
 {id=WavePlayer_0, properties {soundWaveAssetPtr=&/Melee_Hammer_Drop_01}}
 {id=WavePlayer_1, properties {soundWaveAssetPtr=&/Melee_Hammer_Drop_02}}
 {id=WavePlayer_2, properties {soundWaveAssetPtr=&/Melee_Hammer_Drop_03}}
]
properties {
 attenuationSettings=&Fort_Weapon_Drop_Att
 maxDistance=3000
 soundClassObject=&Fort_WeaponMelee
 volumeMultiplier=0.6
}
