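# Sound cue for equipping a melee hammer weapon.
# A random node picks one of three equip wave variations with equal weight, and a modulator at the root applies its default variation to the result.
# It plays in the melee weapon sound class at reduced volume, using the generic player equip attenuation.
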SoundCue /CRD_Melee/Audio/Sounds/Hammer/Melee_Hammer_Equip_Cue
root=Modulator_0
nodes [
 {id=Modulator_0, children [Random_1]}
 {id=Random_1, children [WavePlayer_3,WavePlayer_4,WavePlayer_5], properties {weights [1,1,1]}}
 {id=WavePlayer_3, properties {soundWaveAssetPtr=&/Melee_Hammer_Equip_01}}
 {id=WavePlayer_4, properties {soundWaveAssetPtr=&/Melee_Hammer_Equip_02}}
 {id=WavePlayer_5, properties {soundWaveAssetPtr=&/Melee_Hammer_Equip_03}}
]
properties {
 attenuationSettings=/Game/Sounds/Attenuation/Fort_Player_Equip_Generic_Att
 maxDistance=1800
 soundClassObject=&Fort_WeaponMelee
 volumeMultiplier=0.5
}
