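# Sound cue for the hammer's primary Attack C in the CRD_Melee content, playing one of three Melee_Hammer_Primary_Jump waves.
# An equally weighted Random node selects the variant and feeds a Modulator at the root.
# The cue lowers its volume, routes to the melee weapon sound class, and uses close-range stereo melee attenuation with a shared global melee attack concurrency.
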SoundCue /CRD_Melee/Audio/Sounds/Hammer/Melee_Hammer_Primary_AttackC_Cue
root=Modulator_1
nodes [
 {id=Modulator_1, children [Random_2]}
 {id=Random_2, children [WavePlayer_4,WavePlayer_5,WavePlayer_9], properties {weights [1,1,1]}}
 {id=WavePlayer_4, properties {soundWaveAssetPtr=&/Melee_Hammer_Primary_Jump_01}}
 {id=WavePlayer_5, properties {soundWaveAssetPtr=&/Melee_Hammer_Primary_Jump_02}}
 {id=WavePlayer_9, properties {soundWaveAssetPtr=&/Melee_Hammer_Primary_Jump_03}}
]
properties {
 attenuationSettings=/Game/Sounds/Attenuation/Fort_Weapon_Melee_Attack_Close_Range_Stereo_Att
 maxDistance=10058.75
 soundClassObject=&Fort_WeaponMelee
 volumeMultiplier=0.7
 concurrencySet [/CRD_Melee/Audio/Settings/Melee_Attack_Global_Conc]
}
