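# Sound cue for the hammer's primary melee attack damaging the world, judging by its name and environment concurrency.
# It picks one of three equally weighted wave variants at random and passes it through a modulator for variation.
# The cue plays slightly quieter and higher pitched, in the melee weapon sound class with close-range stereo attenuation.
# Environment and global melee damage concurrency sets limit overlapping impacts.
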SoundCue /CRD_Melee/Audio/Sounds/Hammer/Melee_Hammer_Primary_Damage_World_Cue
root=Modulator_0
nodes [
 {id=Modulator_0, children [Random_0]}
 {id=Random_0, children [WavePlayer_6,WavePlayer_7,WavePlayer_8], properties {weights [1,1,1]}}
 {id=WavePlayer_6, properties {soundWaveAssetPtr=&/Melee_Hammer_Primary_Damage_World_01}}
 {id=WavePlayer_7, properties {soundWaveAssetPtr=&/Melee_Hammer_Primary_Damage_World_02}}
 {id=WavePlayer_8, properties {soundWaveAssetPtr=&/Melee_Hammer_Primary_Damage_World_03}}
]
properties {
 attenuationSettings=/Game/Sounds/Attenuation/Fort_Weapon_Melee_Attack_Close_Range_Stereo_Att
 maxDistance=10058.75
 pitchMultiplier=1.1
 soundClassObject=&Fort_WeaponMelee
 volumeMultiplier=0.9
 concurrencySet [/CRD_Melee/Audio/Settings/Melee_Damage_Environment_Conc, /CRD_Melee/Audio/Settings/Melee_Damage_Global_Conc, null]
}
