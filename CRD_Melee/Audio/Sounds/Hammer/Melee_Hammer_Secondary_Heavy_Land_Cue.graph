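# Sound cue for the landing of a melee hammer's heavy secondary attack.
# A Random node picks one of three equally weighted wave variations and feeds it through a Modulator at the root.
# It plays in the melee weapon sound class with basic foley attenuation and a shared global melee attack concurrency setting.
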SoundCue /CRD_Melee/Audio/Sounds/Hammer/Melee_Hammer_Secondary_Heavy_Land_Cue
root=Modulator_0
nodes [
 {id=Modulator_0, children [Random_0]}
 {id=Random_0, children [WavePlayer_3,WavePlayer_4,WavePlayer_5], properties {weights [1,1,1]}}
 {id=WavePlayer_3, properties {soundWaveAssetPtr=&/Melee_Hammer_Secondary_Heavy_Land_01}}
 {id=WavePlayer_4, properties {soundWaveAssetPtr=&/Melee_Hammer_Secondary_Heavy_Land_02}}
 {id=WavePlayer_5, properties {soundWaveAssetPtr=&/Melee_Hammer_Secondary_Heavy_Land_03}}
]
properties {
 attenuationSettings=&Foley_Attenuation_Basic_Att
 maxDistance=1536
 soundClassObject=&Fort_WeaponMelee
 volumeMultiplier=0.9
 concurrencySet [/CRD_Melee/Audio/Settings/Melee_Attack_Global_Conc]
}
