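# Sound cue for the spinning sound of the melee hammer's whirlwind attack.
# A Random node picks among six equally weighted whirlwind spinning wave variations, and a Modulator node at the root adds playback variation.
# It plays in the weapon melee sound class with close-range stereo melee attack attenuation, a shared global melee attack concurrency limit, and slightly raised pitch and reduced volume.
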SoundCue /CRD_Melee/Audio/Sounds/Hammer/Melee_Hammer_WhirlwindSpinning_Cue
root=Modulator_0
nodes [
 {id=Modulator_0, children [Random_0]}
 {id=Random_0, children [WavePlayer_3,WavePlayer_4,WavePlayer_5,WavePlayer_6,WavePlayer_7,WavePlayer_8], properties {weights [1,1,1,1,1,1]}}
 {id=WavePlayer_3, properties {soundWaveAssetPtr=&/Melee_Hammer_WhirlwindSpinning_01}}
 {id=WavePlayer_4, properties {soundWaveAssetPtr=&/Melee_Hammer_WhirlwindSpinning_02}}
 {id=WavePlayer_5, properties {soundWaveAssetPtr=&/Melee_Hammer_WhirlwindSpinning_03}}
 {id=WavePlayer_6, properties {soundWaveAssetPtr=&/Melee_Hammer_WhirlwindSpinning_04}}
 {id=WavePlayer_7, properties {soundWaveAssetPtr=&/Melee_Hammer_WhirlwindSpinning_05}}
 {id=WavePlayer_8, properties {soundWaveAssetPtr=&/Melee_Hammer_WhirlwindSpinning_06}}
]
properties {
 attenuationSettings=/Game/Sounds/Attenuation/Fort_Weapon_Melee_Attack_Close_Range_Stereo_Att
 maxDistance=10058.75
 pitchMultiplier=1.05
 soundClassObject=&Fort_WeaponMelee
 volumeMultiplier=0.7
 concurrencySet [/CRD_Melee/Audio/Settings/Melee_Attack_Global_Conc]
}
